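# Niagara system for the Soy glider's dive effect, with gameplay and front-end variants chosen through FrontEnd modifiers.
# MeshTrails spawns socket-attached mesh trails, gated in gameplay by full deployment, with exhaust, length and opacity driven by user parameters.
# Flickering blue star sprites sample the skeletal mesh, and speed-stretched speedlines spawn on a cylinder, both rate-scaled by ForwardDot in gameplay.
# The NoseWind mesh emitter is disabled.
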NiagaraSystem /BRCosmetics/Gadgets/Assets/VinderTech_GliderChute/Glider_Soy/Effects/NS_Glider_Soy_Dive
dependencies {
 usedDataInterfaces [&NiagaraDataInterfaceCurve, &NiagaraDataInterfaceVector2DCurve]
 usedDynamicInputs []
 usedModules [&SystemState]
}
emitters {
 FN_MultiStars001 {
  emitterUpdateScript [
   &EmitterState()
   &SpawnRate(
    SpawnRate=&Multiply_Float(
     A=&FN_FrontEndFloatModifier(Gameplay=$User.ForwardDot, Front End=1)
     B=&UniformRangedFloat(Minimum=100, Maximum=200)
    )
    Use Spawn Probability=false
   )
  ]
  particleSpawnScript [
   /Niagara/Modules/Spawn/Initialization/InitializeParticle(
    Color=&UniformRangedLinearColor(Minimum=<rgba 0.095 0.5475 1 0.1>, Maximum=<rgba 0.573 0.9858 1 1>, Color Channel Mode=::NewEnumerator2)
    Lifetime=&UniformRangedFloat(
     Minimum=0.1
     Maximum=&UniformRangedFloat(Minimum=0.2, Maximum=&UniformRangedFloat(Minimum=0.2, Maximum=0.5))
    )
    Module.Write Color=true
    Module.Write Lifetime=true
    Module.Write Mass=true
    Module.Write Material Random=true
    Module.Write Position=false
    Module.Write Scale=false
    Module.Write SpriteRotation=true
    Module.Write SpriteSize=true
    Position=&SimulationPosition()
    Sprite Rotation=&UniformRangedFloat(Minimum=-20, Maximum=20)
    Sprite Size=&Vector2DFromFloat(Value=&UniformRangedFloat(Minimum=4, Maximum=&FN_FrontEndFloatModifier(Gameplay=20, Front End=40)))
   )
   /SetVariables(Particles.PositionMult:Vector3f=&UniformRangedVector(Minimum=<-1 -1 -1>, Maximum=<1 1 1>)) {isSetParameters=true}
   &SkeletalMeshLocation(
    Bone / Socket Index=1
    Inherit Velocity=false
    Interpolate Mesh Sampling=false
    MaskSpawnGroup=false
    Mesh Sampling Type=::NewEnumerator5
    Sampled Position Offset=&Add_Vector(
     A=<17 0 0>
     B=&Multiply_Vector(A=&FN_FrontEndVector3Modifier(Gameplay=<5 25 1>, Front End=<5 10 1>), B=$Particles.PositionMult)
    )
    Sampled Position Offset Coordinate Space:ENiagaraMeshTransforms=::NewEnumerator3
    Skeletal Mesh=$User.SkeletalMesh
    Socket Sampling Mode=::NewEnumerator5
    Write Sampled Mesh Attrributes to Particle Attributes=true
   ) {module=SkeletalMeshLocation001}
   &AddVelocity(Velocity=&FN_FrontEndVector3Modifier(Gameplay=<0 0 0>, Front End=&UniformRangedVector(Minimum=<50 -1 -1>, Maximum=<200 1 1>)))
   &LinearForce(Force=&FN_FrontEndVector3Modifier(
     Gameplay=&Multiply_Vector(
      A=&MakeVector(X=&MakeFloatFromVector(Vector=$User.CurrentExhaust, Channel=::NewEnumerator1), Y=1, Z=1)
      B=&UniformRangedVector(Minimum=<0 0 -2>, Maximum=<-2 500 2>)
     )
     Front End=<0 0 0>
    ))
   &ApplyInitialForces(Apply Force to Position=true, Apply Force to Velocity=false, Apply Rotational Force to Rotational Velocity=false)
  ]
  particleUpdateScript [
   &ParticleState()
   &SkeletalMeshLocation(
    Bone / Socket Index=1
    Inherit Velocity=false
    Interpolate Mesh Sampling=false
    MaskSpawnGroup=false
    Mesh Sampling Type=::NewEnumerator5
    Sampled Position Offset=&Add_Vector(
     A=&Add_Vector(
      A=<17 0 0>
      B=&Multiply_Vector(A=&FN_FrontEndVector3Modifier(Gameplay=<5 26 1>, Front End=<5 10 1>), B=$Particles.PositionMult)
     )
     B=&FN_FrontEndVector3Modifier(
      Gameplay=<0 0 0>
      Front End=&Multiply_VectorByFloat(
       Vector=&TransformVector(Vector=$Particles.Initial.Velocity, Source Space=::World, Destination Space=::Local)
       Float=$Particles.NormalizedAge
      )
     )
    )
    Sampled Position Offset Coordinate Space:ENiagaraMeshTransforms=::NewEnumerator3
    Skeletal Mesh=$User.SkeletalMesh
    Socket Sampling Mode=::NewEnumerator5
    Write Sampled Mesh Attrributes to Particle Attributes=true
   )
   /SetVariables(Particles.Position=&FN_FrontEndVector3Modifier(
     Gameplay=&Add_Vector(A=Linked() {linkedVariable {name=Particles.Initial.Position, type=NiagaraPosition}}, B=$Particles.SkeletalMeshLocation.SampledPosition)
     Front End=Linked() {linkedVariable {name=Particles.Position, type=NiagaraPosition}}
    )) {isSetParameters=true}
   &ScaleSpriteSize(Scale Factor=&Vector2DFromFloat(Value=&Multiply_Float(
      A=&FloatFromCurve(FloatCurve=DataInterface() {propertyValues {Curve {keys [<at 0 1>,<at 0.183 0.1383>,<at 0.3114 0.9388>,<at 0.411 0.2817>,<at 0.4973 0.4591>,<at 0.6394 0.0099>,<at 0.8 0.9523>,<at 0.9 0.0712>,<at 1 1>]}}})
      B=&FloatFromCurve(FloatCurve=DataInterface() {propertyValues {Curve {keys [
           {arriveTangent=0.0702, interpMode=RCIM_Cubic, leaveTangent=0.0702, tangentMode=RCTM_User, time=0, value=1}
           {arriveTangent=-2.288, interpMode=RCIM_Cubic, leaveTangent=-2.288, tangentMode=RCTM_User, time=1, value=0}
          ]}}})
     ))) {module=ScaleSpriteSize001}
   &SolveForcesAndVelocity(Clamp Velocity=false, Limit Acceleration=false)
  ]
  rendererClasses [&NiagaraSpriteRendererProperties]
  renderers [{rendererClass=&NiagaraSpriteRendererProperties}]
 }
 MeshTrails {
  emitterSpawnScript [&FN_PreserveSkydiveUserVariables() {module=PreserveUserVariables}]
  emitterUpdateScript [
   &EmitterState()
   &SpawnBurst_Instantaneous(
    Spawn Count=HlslExpression() {hlslExpression="Emitter.LoopCount == 0"}
    Spawn Probability=&FN_FrontEndFloatModifier(
     Gameplay=&MakeCustomFloatFromBool(Bool=$User.bIsFullyDeployed, True Float=1, False Float=0)
     Front End=1
    )
    Use Spawn Probability=true
   ) {module=SpawnBurst_Instantaneous001}
  ]
  particleSpawnScript [
   &InitializeParticle(
    Module.Write Color=true
    Module.Write Lifetime=true
    Module.Write Mass=false
    Module.Write Material Random=true
    Module.Write Position=false
    Module.Write Scale=false
    Module.Write SpriteRotation=false
    Module.Write SpriteSize=false
   )
   &self:Mesh_Trail_Setup(
    Full Trail Length=500
    Long Trail Width:NiagaraFloat=&FN_FrontEndFloatModifier(Gameplay=50, Front End=37.5)
    Short Trail Multiplier=0.35
    ShortTrail Width:NiagaraFloat=&FN_FrontEndFloatModifier(Gameplay=40, Front End=30)
   ) {module=Mesh_Trail_Setup}
  ]
  particleUpdateScript [
   &ParticleState(Kill Particles When Lifetime Has Elapsed=false, Let Infinitely Lived Particles Die When Emitter Deactivates=true)
   &SkeletalMeshLocation(
    Apply Sampled Orientation to Mesh Particles=true
    Bone / Socket Index=1
    Inherit Velocity=false
    Interpolate Mesh Sampling=false
    MaskSpawnGroup=false
    Mesh Sampling Type=::NewEnumerator5
    Sampled Position Offset Coordinate Space:ENiagaraMeshTransforms=::NewEnumerator3
    Skeletal Mesh=$User.SkeletalMesh
    Socket Sampling Mode=::NewEnumerator5
    Write Sampled Mesh Attrributes to Particle Attributes=true
   )
   /SetVariables(Particles.MeshTwinOrientation:Quat4f=&MultiplyQuaternion(Quaternion A=$Particles.MeshOrientation, Quaternion B=<1 0 0 0>)) {isSetParameters=true}
   /SetVariables(Particles.NormalishSocketVelocity:NiagaraFloat=&Multiply_Float(
     A=-0.004
     B=&ClampFloat(
      Float=&MakeFloatFromVector(Vector=$Particles.SkeletalMeshLocation.SampledVelocity, Channel=::NewEnumerator0)
      Min=-250
      Max=250
     )
    )) {isSetParameters=true}
   /SetVariables(Particles.VelocityOutputControl:NiagaraFloat=&Subtract_Float(
     A=2.7
     B=&Multiply_Float(A=2.2725, B=&Multiply_Float(A=$Particles.NormalishSocketVelocity, B=$Particles.NormalishSocketVelocity))
    )) {isSetParameters=true}
   &self:Exhaust_to_DynaMat(
    ExhaustVector:Vector3f=&FN_FrontEndVector3Modifier(
     Gameplay=$User.CurrentExhaust
     Front End=&self:Interpolate_Vector(
      Rate Of Change:NiagaraFloat=10
      Vector to Interpolate:Vector3f=&TransformVector(Vector=$Particles.SkeletalMeshLocation.SampledVelocity, Source Space=::World, Destination Space=::Local)
      Velocity Multiplier:Vector3f=<0 30 0>
     )
    )
    SpeedMult=1
   ) {module=Exhaust_to_DynaMat}
   &self:Angle_to_DynaMat(Exhaust Angle:NiagaraFloat=$User.ExhaustAngle, SpeedMult=1) {enabled=false, module=Angle_to_DynaMat}
   &self:Mesh_Trail_Length_Update(Dynamic Length Scale:NiagaraFloat=&FN_FrontEndFloatModifier(Gameplay=$User.TrailLengthScale, Front End=1)) {module=Mesh_Trail_Length_Update}
   &ScaleColor(
    Scale Alpha=&FN_FrontEndFloatModifier(Gameplay=$User.BPOpacityMult, Front End=1)
    ScaleA=true
    ScaleRGB=false
   )
  ]
  rendererClasses [&NiagaraMeshRendererProperties]
  renderers [
   {rendererClass=&NiagaraMeshRendererProperties, rendererIndex=0}
   {rendererClass=&NiagaraMeshRendererProperties, rendererIndex=1}
   {rendererClass=&NiagaraMeshRendererProperties, rendererIndex=2}
  ]
 }
 NoseWind {
  bEnabled=false
  emitterSpawnScript [&FN_PreserveSkydiveUserVariables() {module=PreserveUserVariables}]
  emitterUpdateScript [
   &EmitterState()
   &SpawnRate(
    Spawn Probability=&MakeCustomFloatFromBool(Bool=$User.bIsFullyDeployed, True Float=1, False Float=0)
    SpawnRate=&Multiply_Float(A=&FN_FrontEndFloatModifier(Gameplay=$User.ForwardDot, Front End=0), B=4)
    Use Spawn Probability=true
   )
  ]
  particleSpawnScript [
   &InitializeParticle(
    Lifetime=0.35
    Mesh Scale=<1.3 0.65 0.6>
    Module.Write Color=true
    Module.Write Lifetime=true
    Module.Write Mass=false
    Module.Write Material Random=true
    Module.Write Position=false
    Module.Write Scale=true
    Module.Write SpriteRotation=false
    Module.Write SpriteSize=false
   )
   /Niagara/Modules/Update/Orientation/V2/OrientMeshToVector(Look At Direction:Vector3f=<0 -1 0>, Override Quat=false, UseOptionalSideAxis=false, UseOptionalUpAxis=false)
  ]
  particleUpdateScript [
   &ParticleState()
   &SkeletalMeshLocation(
    Inherit Velocity=false
    Interpolate Mesh Sampling=false
    MaskSpawnGroup=false
    Mesh Sampling Type=::NewEnumerator5
    Sampled Position Offset=<0 42 2>
    Sampled Position Offset Coordinate Space:ENiagaraMeshTransforms=::NewEnumerator3
    Skeletal Mesh=$User.SkeletalMesh
    Socket Sampling Mode=::NewEnumerator5
    Write Sampled Mesh Attrributes to Particle Attributes=true
   )
   &ScaleColor(
    Scale Alpha=&FloatFromCurve(FloatCurve=DataInterface() {propertyValues {Curve {keys [
         <at 0 0 RCIM_Cubic>
         {arriveTangent=0.855, interpMode=RCIM_Cubic, leaveTangent=0.855, tangentMode=RCTM_User, time=0.2, value=1}
         {arriveTangent=-8.8264, interpMode=RCIM_Cubic, leaveTangent=-8.8264, tangentMode=RCTM_User, time=1, value=0}
        ]}}})
    ScaleA=true
    ScaleRGB=true
   )
  ]
  rendererClasses [&NiagaraMeshRendererProperties]
  renderers [{rendererClass=&NiagaraMeshRendererProperties, rendererIndex=0}]
 }
 Speedlines {
  emitterUpdateScript [
   &EmitterState(Enable Distance Culling=true, MaxDistance=2048, Module.bUseMaxDistance=true, Module.bUseMinDistance=false, Scalability Mode=::NewEnumerator1)
   &SpawnRate(
    SpawnRate=&Multiply_Float(
     A=&FN_FrontEndFloatModifier(Gameplay=$User.ForwardDot, Front End=1)
     B=&UniformRangedFloat(Minimum=75, Maximum=150)
    )
    Use Spawn Probability=false
   )
  ]
  particleSpawnScript [
   &InitializeParticle(
    Color=&MakeLinearColorFromVectorAndFloat(
     "Vector (RGB)"=<1 1 1>
     "Float (Alpha)"=&Multiply_Float(
      A=&UniformRangedFloat(Minimum=0.015, Maximum=0.15)
      B=&Add_Float(
       A=1
       B=&Multiply_Float(A=10, B=&Subtract_Float(A=1, B=$Engine.Emitter.SpawnCountScale))
      )
     )
    )
    Lifetime=&FN_FrontEndFloatModifier(
     Gameplay=&UniformRangedFloat(Minimum=0.2, Maximum=0.35)
     Front End=&UniformRangedFloat(Minimum=0.1, Maximum=0.2)
    )
    Module.Write Color=true
    Module.Write Lifetime=true
    Module.Write Mass=true
    Module.Write Material Random=true
    Module.Write Position=false
    Module.Write Scale=false
    Module.Write SpriteRotation=false
    Module.Write SpriteSize=true
    Position=&SimulationPosition()
    Sprite Size=&UniformRangedVector2D(Minimum=<3 4>, Maximum=<5 6>)
   )
   &CylinderLocation(
    Cylinder Radius=&UniformRangedFloat(Minimum=100, Maximum=150)
    MaskSpawnGroup=false
    Module.Enable Lathe Profile=false
    Module.Override Randomness=false
    Module.Surface Only=true
    Orientation Axis:ENiagaraOrientationAxis=::XAxis
    Use Endcaps In Surface Only Mode=false
   )
   &AddVelocity(Velocity=&FN_FrontEndVector3Modifier(Gameplay=<0 1000 0>, Front End=<0 3000 0>))
   &AddVelocityFromPoint(
    Module.Use Falloff Distance=false
    Velocity Strength=&Multiply_Float(
     A=&FN_FrontEndFloatModifier(Gameplay=1, Front End=2)
     B=&UniformRangedFloat(Minimum=500, Maximum=700)
    )
   )
  ]
  particleUpdateScript [
   &ParticleState()
   &SolveForcesAndVelocity(Clamp Velocity=false, Limit Acceleration=false)
   &ScaleSpriteSize(Scale Factor=&Vector2DFromCurve(
     Vector2Curve=DataInterface() {propertyValues {XCurve {keys [<at 0 0.5 RCIM_Cubic>,<at 0.5 1 RCIM_Cubic>]}, YCurve {keys [<at 0 0 RCIM_Cubic>,<at 0.5 1 RCIM_Cubic>]}}}
     ScaleCurve=&Vector2DFromFloat(Value=1)
    ))
   &ScaleSpriteSizeBySpeed(Max Scale Factor=<1 40>, Min Scale Factor=<0.01 0.05>, Velocity Threshold=1)
   &ScaleColor(
    Scale Alpha=&FloatFromCurve(FloatCurve=DataInterface() {propertyValues {Curve {keys [<at 0 0 RCIM_Cubic>, <at 0.2 1>, {arriveTangent=-2.221, interpMode=RCIM_Cubic, leaveTangent=-2.221, tangentMode=RCTM_User, time=1, value=0}]}}})
    ScaleA=true
    ScaleRGB=true
   )
  ]
  rendererClasses [&NiagaraSpriteRendererProperties]
  renderers [{rendererClass=&NiagaraSpriteRendererProperties}]
 }
}
summary {
 systemName=NS_Glider_Soy_Dive
 emitters [
  {emitterName=Speedlines, rendererClasses [&NiagaraSpriteRendererProperties]}
  {bEnabled=false, emitterName=NoseWind, rendererClasses [&NiagaraMeshRendererProperties]}
  {emitterName=MeshTrails, rendererClasses [&NiagaraMeshRendererProperties]}
  {emitterName=FN_MultiStars001, rendererClasses [&NiagaraSpriteRendererProperties]}
 ]
 userVariables [
  User.bIsFrontEnd {type=NiagaraBool}
  User.bIsFrontEndPreview {type=NiagaraBool}
  User.bIsFullyDeployed {type=NiagaraBool}
  User.BPOpacityMult {type=NiagaraFloat, defaultValue=0, description=""}
  User.Color {type=LinearColor, description="", defaultValue {struct=LinearColor, value=<rgba 1 1 1 1>}}
  User.CurrentExhaust {type=Vector3f, description="", defaultValue {struct=Vector3f, value=<0 0 0>}}
  User.ExhaustAngle {type=NiagaraFloat, defaultValue=0, description=""}
  User.ForwardDot {type=NiagaraFloat, defaultValue=0, description=""}
  User.LocalAlphaRadial {type=NiagaraFloat, defaultValue=0.25, description=""}
  User.PlayerSpeed {type=NiagaraFloat, defaultValue=0, description=""}
  User.RibbonIDOffset {type=NiagaraInt32}
  User.SizeScale {type=NiagaraFloat, defaultValue=0, description=""}
  User.SkeletalMesh {
   type=NiagaraDataInterfaceSkeletalMesh
   defaultValue {
    struct=NiagaraExt_VariableValue_DataInterface
    value {dataInterface=&self:NiagaraDataInterfaceSkeletalMesh_0}
   }
  }
  User.TrailLengthScale {type=NiagaraFloat, defaultValue=0, description=""}
 ]
}
